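# Sound cue for the Office emote's looping audio, named as the 1P foley variant.
# A Random node picks between two looping Office wave files with equal weight.
# Despite the Foley name, it uses the emote music sound class and concurrency rule, with shared Athena emote attenuation.
# It also has a send to the music audio-analysis submix.
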SoundCue /BRCosmetics/Athena/Sounds/Emotes/Office/SC_EmoteFoley_Office_Loop_1P
root=Random_1
nodes [
 {id=Random_1, children [WavePlayer_0,WavePlayer_1], properties {weights [1,1]}}
 {id=WavePlayer_0, properties {bLooping=true, soundWaveAssetPtr=/Game/Athena/Sounds/Emotes/Office/Emote_Office_Loop_02}}
 {id=WavePlayer_1, properties {bLooping=true, soundWaveAssetPtr=/Game/Athena/Sounds/Emotes/Office/Emote_Office_Loop_01}}
]
properties {
 attenuationSettings=&Fort_Emotes_Athena_Att
 maxDistance=1600
 soundClassObject=&Fort_Music_Emote
 volumeMultiplier=1
 concurrencySet [/Game/Sounds/Concurrency_Rules/Athena_EmoteMusic_Conc]
 soundSubmixSends [{minSendLevel=0, soundSubmix=&Fortnite_AudioAnalysis_Music}]
}
